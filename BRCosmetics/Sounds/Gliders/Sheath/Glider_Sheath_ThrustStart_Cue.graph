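# Sound cue for the thrust start of the Sheath glider in Battle Royale.
# A Random node picks one of three equally weighted Sheath thrust waves, and a Modulator slightly varies its volume.
# The cue plays at a reduced volume multiplier through glider thrust attenuation, a shared thrust concurrency set and the Fort_Gliders_Land sound class.
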SoundCue /BRCosmetics/Sounds/Gliders/Sheath/Glider_Sheath_ThrustStart_Cue
root=Modulator_0
nodes [
 {id=Modulator_0, children [Random_1], properties {volumeMax=0.7, volumeMin=0.65}}
 {id=Random_1, children [WavePlayer_0,WavePlayer_2,WavePlayer_4], properties {weights [1,1,1]}}
 {id=WavePlayer_0, properties {soundWaveAssetPtr=&/Athena_Glider_Sheath_Thrust_01}}
 {id=WavePlayer_2, properties {soundWaveAssetPtr=&/Athena_Glider_Sheath_Thrust_02}}
 {id=WavePlayer_4, properties {soundWaveAssetPtr=&/Athena_Glider_Sheath_Thrust_03}}
]
properties {
 attenuationSettings=&Fort_Player_Glider_Thrust_Athena_Att
 maxDistance=12500
 soundClassObject=&Fort_Gliders_Land
 volumeMultiplier=0.2
 concurrencySet [&Athena_Glider_Thrust_Conc]
}
