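# Sound cue for the Sheath glider's looping in-air thrust audio in Battle Royale cosmetics.
# A looping wave player plays the Sheath in-air loop through a modulator that adds a slight random pitch variation at constant volume.
# Playback uses the shared glider thrust attenuation, the cosmetic glider loop audio properties sheet and the glider landing sound class at reduced volume.
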SoundCue /BRCosmetics/Sounds/Gliders/Sheath/Glider_Sheath_Thrust_Loop_Cue
root=Modulator_0
nodes [
 {id=Modulator_0, children [WavePlayer_1], properties {pitchMax=1.01, pitchMin=0.99, volumeMax=1, volumeMin=1}}
 {id=WavePlayer_1, properties {bLooping=true, soundWaveAssetPtr=&/Athena_Glider_Sheath_InAirLoop}}
]
properties {
 attenuationSettings=&Fort_Player_Glider_Thrust_Athena_Att
 audioPropertiesSheet=&APS_Cosmetic_Glider_Loop
 maxDistance=12500
 soundClassObject=&Fort_Gliders_Land
 volumeMultiplier=0.5
}
